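; Big rig Valet vehicle Blueprint that swaps one skeletal mesh material to an occluded variant while the local player rides as passenger.
; BeginPlay binds seat enter/exit dispatchers, which also cover seat swaps; seat index 1 entry stores the passenger, and exits check that stored pawn.
; The swap picks from regular or occluded arrays using the same random index.
; It also enlarges water-entry splash FX and loads hotfixable pontoon tick-rate scales during construction.
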(event EventOnEnteredWaterBody
  (params (WaterBody :type Fort%20Water%20Body%20Actor%20Object%20Reference))
  (|CallParentFunction WaterBody)
  (bind _ns_largesplashfx (Variables|WaterInteraction|GetNS_LargeSplashFX)
    (:then
      (Niagara|SetNiagaraVariable(Float) _ns_largesplashfx "EffectScale" 1.4))
    (:else)))

(event EventBeginPlay
  ;[1 1 1] Bind to Enter/Exit Vehicle Seats because Enter and Exit vehicle doesn't trigger on seat swap anymore
  (|CallParentFunction)
  (Vehicle|AssignOnPawnExitedVehicleSeat (EventDispatchers|CreateEvent :!event Custom|PawnExitVehicleSeat))
  (Vehicle|AssignOnPawnEnteredVehicleSeat (EventDispatchers|CreateEvent :!event Custom|PawnEnterVehicleSeat)))

(event Custom|PawnEnterVehicleSeat
  (params (Vehicle :type Fort%20Vehicle%20Interface%20Interface%20%28by%20ref%29) (PlayerPawn :type Fort%20Pawn%20Object%20Reference) (SeatIndex :type Integer))
  ;[1 1 1] Swap to Occluded Material
  (if (== SeatIndex 1)
    (bind _output_get (Variables|Default|SetPawnPassenger PlayerPawn))
    (CallFunction|PassengerEnter _output_get)))

(event Custom|PawnExitVehicleSeat
  (params (Vehicle :type Fort%20Vehicle%20Interface%20Interface%20%28by%20ref%29) (PlayerPawn :type Fort%20Pawn%20Object%20Reference) (SeatIndex :type Integer))
  (bind _pawnpassenger (Variables|Default|GetPawnPassenger))
  ;[1 1 1] Swap back to Regular Material
  (Utilities|IsValid _pawnpassenger
    (:Is%20Valid
      (CallFunction|PassengerExit _pawnpassenger))))

(fn LocalPassengerMatSwap
  (Rendering|Material|SetMaterial (Variables|Vehicle|GetSkeletalMesh) 4 (select (Variables|Default|GetIsLocalPlayerPassenger) (Utilities|Array|Get(acopy) (Variables|Default|GetSK_RandomMat0_Array_Occluded) (Variables|Default|GetRandomMatInt)) (Utilities|Array|Get(acopy) (Variables|Default|GetSK_RandomMat0_Array) (Variables|Default|GetRandomMatInt)))))

(fn PassengerEnter
  (params (PlayerPawn :type Fort%20Pawn%20Object%20Reference))
  (if (Pawn|IsLocallyControlled PlayerPawn)
    (Variables|Default|SetIsLocalPlayerPassenger true)
    (CallFunction|LocalPassengerMatSwap)))

(fn PassengerExit
  (params (PlayerPawn :type Fort%20Pawn%20Object%20Reference))
  (if (Pawn|IsLocallyControlled PlayerPawn)
    (Variables|Default|SetIsLocalPlayerPassenger false)
    (CallFunction|LocalPassengerMatSwap)
    (Variables|Default|SetPawnPassenger 0)))

(fn ConstructionScript
  (|CallParentFunction)
  (Variables|PontoonScaling|SetScalePontoonTickRateFront (ScalableFloat|GetValueatLevel (Variables|Default|GetHotfixScalePontoonTickRateFront_0)))
  (Variables|PontoonScaling|SetScalePontoonTickRateRear (ScalableFloat|GetValueatLevel (Variables|Default|GetHotfixScalePontoonTickRateRear_0))))
